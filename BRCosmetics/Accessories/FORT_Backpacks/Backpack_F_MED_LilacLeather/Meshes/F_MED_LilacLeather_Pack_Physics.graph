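# Physics asset for the LilacLeather backpack cosmetic, giving its dangling chains and kunai capsule bodies for secondary motion.
# Left and right five-link chains each end in a kunai, while a shorter middle chain has two segments, all rooted through base bodies on the pack.
# Side bases rotate freely against the pack; other joints have angular limits with little twist.
# The pack has two capsules, and a large Spelvis capsule has no constraint.
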PhysicsAsset /BRCosmetics/Accessories/FORT_Backpacks/Backpack_F_MED_LilacLeather/Meshes/F_MED_LilacLeather_Pack_Physics
bodies {
 Base_chain_l [{shapeType=Capsule, length=0.505, radius=0.505, shapeName=Base_chain_l_capsule, rotation=<rot 0 -90 0>}]
 Base_chain_m [{shapeType=Capsule, length=0.505, radius=0.505, shapeName=Base_chain_m_capsule, rotation=<rot 0 -90 0>}]
 Base_chain_r [{shapeType=Capsule, length=0.505, radius=0.505, shapeName=Base_chain_r_capsule, rotation=<rot 0 -90 180>}]
 Dyn_Kunai_R [{shapeType=Capsule, length=7.3905, radius=2.7537, shapeName=Dyn_Kunai_r_capsule, center=<6.1646 0 -0.5393>, rotation=<rot 85 0.0002 0>}]
 Dyn_Kunai_l [{
   shapeType=Capsule
   length=7.3905
   radius=2.7537
   shapeName=Dyn_Kunai_l_capsule
   center=<7.3829 -0.0952 -0.566>
   rotation=<rot 0 -90.1105 -86.6615>
  }]
 Dyn_chain_l_01 [{
   shapeType=Capsule
   length=1.844
   radius=0.8616
   shapeName=Dyn_chain_l_01_capsule
   center=<1.8208 0.0586 -0.1222>
   rotation=<rot 0 -88.1172 -86.7198>
  }]
 Dyn_chain_l_02 [{
   shapeType=Capsule
   length=1.8157
   radius=0.8544
   shapeName=Dyn_chain_l_02_capsule
   center=<1.7971 0.0515 -0.0739>
   rotation=<rot 0 -90.0211 -85.1554>
  }]
 Dyn_chain_l_03 [{
   shapeType=Capsule
   length=0.505
   radius=0.505
   shapeName=Dyn_chain_l_03_capsule
   center=<1.8256 0.0784 -0.2021>
   rotation=<rot 0 -90.5866 -86.1751>
  }]
 Dyn_chain_l_04 [{
   shapeType=Capsule
   length=1.8193
   radius=0.8427
   shapeName=Dyn_chain_l_04_capsule
   center=<1.7746 0.0454 -0.0668>
   rotation=<rot 0 -89.4867 -84.3455>
  }]
 Dyn_chain_l_05 [{
   shapeType=Capsule
   length=0.505
   radius=0.505
   shapeName=Dyn_chain_l_05_capsule
   center=<1.8018 -0.1378 -0.1958>
   rotation=<rot 0 -91.406 -86.1665>
  }]
 Dyn_chain_m_01 [{shapeType=Capsule, length=9.2576, radius=4.7901, shapeName=Dyn_chain_m_01_capsule, center=<6.224 -0.1741 0>, rotation=<rot 0 -94.9998 0>}]
 Dyn_chain_m_02 [{shapeType=Capsule, length=6.005, radius=3.505, shapeName=Dyn_chain_m_02_capsule, center=<10.2756 0 0>, rotation=<rot 0 -90 0>}]
 Dyn_chain_r_01 [{shapeType=Capsule, length=1.844, radius=0.8616, shapeName=Dyn_chain_r_01_capsule}]
 Dyn_chain_r_02 [{shapeType=Capsule, length=1.8157, radius=0.8544, shapeName=Dyn_chain_r_02_capsule}]
 Dyn_chain_r_03 [{shapeType=Capsule, length=0.505, radius=0.505, shapeName=Dyn_chain_r_03_capsule}]
 Dyn_chain_r_04 [{shapeType=Capsule, length=1.8193, radius=0.8427, shapeName=Dyn_chain_r_04_capsule}]
 Dyn_chain_r_05 [{shapeType=Capsule, length=0.505, radius=0.505, shapeName=Dyn_chain_r_05_capsule}]
 Spelvis [{shapeType=Capsule, length=14, radius=9.505, shapeName=Spelvis_capsule, center=<0.4601 -8.8068 0.3926>, rotation=<rot -0.0004 -90 0>}]
 pack [
  {
   shapeType=Capsule
   length=18.9971
   radius=4.1722
   shapeName=pack_capsule
   center=<-0.0011 -3.2509 -8.5922>
   rotation=<rot 0 -90.0023 -89.9956>
  }
  {shapeType=Capsule, length=45.5, radius=7.5, shapeName=pack_capsule, center=<0 8.0959 -18.123>, rotation=<rot -90 -0.0003 0>}
 ]
}
constraints [
 {
  bone1Name=Base_chain_l
  bone2Name=pack
  swing1LimitDegrees=45
  swing1Motion=Free
  swing2LimitDegrees=45
  swing2Motion=Free
  twistLimitDegrees=45
  twistMotion=Free
 }
 {bone1Name=Dyn_chain_l_01, bone2Name=Base_chain_l, swing1LimitDegrees=5, swing2LimitDegrees=5, twistLimitDegrees=5}
 {bone1Name=Dyn_chain_l_02, bone2Name=Dyn_chain_l_01, swing1LimitDegrees=30, swing2LimitDegrees=20, twistLimitDegrees=5}
 {bone1Name=Dyn_chain_l_03, bone2Name=Dyn_chain_l_02, swing1LimitDegrees=30, swing2LimitDegrees=20, twistLimitDegrees=5}
 {bone1Name=Dyn_chain_l_04, bone2Name=Dyn_chain_l_03, swing1LimitDegrees=30, swing2LimitDegrees=20, twistLimitDegrees=5}
 {bone1Name=Dyn_chain_l_05, bone2Name=Dyn_chain_l_04, swing1LimitDegrees=30, swing2LimitDegrees=10, twistLimitDegrees=5}
 {bone1Name=Dyn_Kunai_l, bone2Name=Dyn_chain_l_05, swing1LimitDegrees=30, swing2LimitDegrees=10, twistLimitDegrees=5}
 {
  bone1Name=Base_chain_r
  bone2Name=pack
  swing1LimitDegrees=45
  swing1Motion=Free
  swing2LimitDegrees=45
  swing2Motion=Free
  twistLimitDegrees=45
  twistMotion=Free
 }
 {bone1Name=Dyn_chain_r_01, bone2Name=Base_chain_r, swing1LimitDegrees=5, swing2LimitDegrees=5, twistLimitDegrees=5}
 {bone1Name=Dyn_chain_r_02, bone2Name=Dyn_chain_r_01, swing1LimitDegrees=30, swing2LimitDegrees=20, twistLimitDegrees=5}
 {bone1Name=Dyn_chain_r_03, bone2Name=Dyn_chain_r_02, swing1LimitDegrees=30, swing2LimitDegrees=20, twistLimitDegrees=5}
 {bone1Name=Dyn_chain_r_04, bone2Name=Dyn_chain_r_03, swing1LimitDegrees=30, swing2LimitDegrees=20, twistLimitDegrees=5}
 {bone1Name=Dyn_chain_r_05, bone2Name=Dyn_chain_r_04, swing1LimitDegrees=30, swing2LimitDegrees=10, twistLimitDegrees=5}
 {bone1Name=Dyn_Kunai_R, bone2Name=Dyn_chain_r_05, swing1LimitDegrees=27, swing2LimitDegrees=10, twistLimitDegrees=10}
 {bone1Name=Base_chain_m, bone2Name=pack, swing1LimitDegrees=5, swing2LimitDegrees=5, twistLimitDegrees=5}
 {bone1Name=Dyn_chain_m_01, bone2Name=Base_chain_m, swing1LimitDegrees=15, swing2LimitDegrees=15, twistLimitDegrees=5}
 {bone1Name=Dyn_chain_m_02, bone2Name=Dyn_chain_m_01, swing1LimitDegrees=30, swing2LimitDegrees=30, twistLimitDegrees=5}
]
